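# Physics asset for the F_MED_Tactical_Red female player body, covering the dynamic shoelace on the right leg only.
# Inner and outer two-bone capsule chains hang from a tiny master body, each joint using tight twist and moderate swing limits.
# The master body is constrained to a large right lower-calf capsule with swing and twist left free.
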PhysicsAsset /BRCosmetics/Characters/Player/Female/Medium/Bodies/F_MED_Tactical_Red/Meshes/F_MED_Tactical_Red_Dyn
bodies {
 calf_twist_02_r [{shapeType=Capsule, length=22.0543, radius=4.5953, center=<10.0262 1.9533 -0.5059>, rotation=<rot 90 0 0>}]
 dyn_shoelace_in_r_01 [{shapeType=Capsule, length=2.1578, radius=0.8422, center=<2 0 0>, rotation=<rot 90 0 0>}]
 dyn_shoelace_in_r_02 [{shapeType=Capsule, length=2, radius=1, center=<2 0 0>, rotation=<rot 90 0 0>}]
 dyn_shoelace_out_r_01 [{shapeType=Capsule, length=2.1995, radius=0.8005, center=<2 0 0>, rotation=<rot 90 0 0>}]
 dyn_shoelace_out_r_02 [{shapeType=Capsule, length=2, radius=1, center=<2 0 0>, rotation=<rot 90 0 0>}]
 dyn_shoelace_r_MSTR [{shapeType=Capsule, length=0.2, radius=0.2, rotation=<rot 90 0 0>}]
}
constraints [
 {bone1Name=dyn_shoelace_out_r_01, bone2Name=dyn_shoelace_r_MSTR, swing1LimitDegrees=30, swing2LimitDegrees=30, twistLimitDegrees=5}
 {bone1Name=dyn_shoelace_out_r_02, bone2Name=dyn_shoelace_out_r_01, swing1LimitDegrees=30, swing2LimitDegrees=30, twistLimitDegrees=5}
 {bone1Name=dyn_shoelace_in_r_01, bone2Name=dyn_shoelace_r_MSTR, swing1LimitDegrees=30, swing2LimitDegrees=30, twistLimitDegrees=5}
 {bone1Name=dyn_shoelace_in_r_02, bone2Name=dyn_shoelace_in_r_01, swing1LimitDegrees=30, swing2LimitDegrees=30, twistLimitDegrees=5}
 {
  bone1Name=dyn_shoelace_r_MSTR
  bone2Name=calf_twist_02_r
  swing1LimitDegrees=45
  swing1Motion=Free
  swing2LimitDegrees=45
  swing2Motion=Free
  twistLimitDegrees=45
  twistMotion=Free
 }
]
