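# Physics asset for the dynamic arm bones of the F_MED_Tactical_Red female player body, defining secondary motion for an armband knot and its tails.
# Each of the six armband bones has a single capsule body.
# Constraints form a branching chain: the knot hangs from a master bone and splits into two-segment high and low tails.
# Swing limits are widest at the knot and tighter along the tails, while every joint allows very little twist.
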PhysicsAsset /BRCosmetics/Characters/Player/Female/Medium/Bodies/F_MED_Tactical_Red/Meshes/F_MED_Tactical_Red_Dyn_arm
bodies {
 dyn_armband_hi_01 [{shapeType=Capsule, length=1.3356, radius=0.7162, center=<0.8 0 0>, rotation=<rot 90 0 0>}]
 dyn_armband_hi_02 [{shapeType=Capsule, length=1.5639, radius=0.5719, center=<0.714 0.6119 -0.7956>, rotation=<rot 45.5937 37.4732 5.4343>}]
 dyn_armband_knot [{shapeType=Capsule, length=2, radius=0.7162, center=<0.05 0 0>, rotation=<rot 90 0 0>}]
 dyn_armband_knot_MSTR [{shapeType=Capsule, length=0.505, radius=0.505, rotation=<rot 0 -90 0>}]
 dyn_armband_low_01 [{shapeType=Capsule, length=2, radius=0.7162, center=<0.8 0 0>, rotation=<rot 90 0 0>}]
 dyn_armband_low_02 [{shapeType=Capsule, length=2, radius=0.7162, center=<0.8 0 0>, rotation=<rot 90 0 0>}]
}
constraints [
 {bone1Name=dyn_armband_knot, bone2Name=dyn_armband_knot_MSTR, swing1LimitDegrees=45, swing2LimitDegrees=45, twistLimitDegrees=2}
 {bone1Name=dyn_armband_low_01, bone2Name=dyn_armband_knot, swing1LimitDegrees=35, swing2LimitDegrees=35, twistLimitDegrees=2}
 {bone1Name=dyn_armband_low_02, bone2Name=dyn_armband_low_01, swing1LimitDegrees=35, swing2LimitDegrees=35, twistLimitDegrees=2}
 {bone1Name=dyn_armband_hi_01, bone2Name=dyn_armband_knot, swing1LimitDegrees=35, swing2LimitDegrees=28, twistLimitDegrees=2}
 {bone1Name=dyn_armband_hi_02, bone2Name=dyn_armband_hi_01, swing1LimitDegrees=35, swing2LimitDegrees=19, twistLimitDegrees=2}
]
